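# Anim Blueprint for the Elevate wingsuit glider, animating deploy, glide and close from the player's skydiving state.
# The Locomotion state machine feeds lean, turn, aim and jitter additives.
# The root bone then snaps to the player's world-space spine_05 with an offset; Locomotion_GliderItem is not wired to the output.
# State notifies hide or unhide Glider_Base, with a matching Elevate backpack selecting the deploy blendspace and skipping the unhide delay.
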AnimBlueprint /BRCosmetics/Athena/Items/Cosmetics/Gliders/CustomWingSuitGlider/Gadgets/Assets/VinderTech_WingSuitChute/Blueprints/Elevate_WingSuit_AnimBP
target_skeleton=/BRCosmetics/Gadgets/Assets/VinderTech_GliderChute/Glider_Elevate/Meshes/Elevate_Glider_Skeleton
graphs [
 AnimGraph {nodes [
   n1 = BreakTransform(InTransform=Get Player_Spine05() -> (Player_Spine05:Transform))
   n2 = "Transform (Modify) Bone - Bone: Root"(ComponentPose=Local To Component(LocalPose=Use cached pose 'PreAttachMOd'()), Alpha=1, Translation=@n1, Rotation=@n1.Rotation, Scale="1,1,1") {
    rotationMode=BMM_Replace
    rotationSpace=BCS_WorldSpace
    translationMode=BMM_Replace
    translationSpace=BCS_WorldSpace
    alphaBoolBlend {blendOption=Linear}
    boneToModify {boneName=Root}
   }
   n3 = "Transform (Modify) Bone - Bone: Root"(ComponentPose=@n2, Alpha=1, Translation="0,-10,0", Rotation="-90,30,-30", Scale="1,1,1") {
    rotationMode=BMM_Additive
    rotationSpace=BCS_BoneSpace
    translationMode=BMM_Additive
    alphaBoolBlend {blendOption=Linear}
    boneToModify {boneName=Root}
    rotation=<rot -90 30 -30>
    translation {y=-10}
   }
   Output Pose(Result=Slot 'DefaultSlot'(Source=Component To Local(ComponentPose=@n3))) {layerGroup=None}
   n4 = Break Fort Anim Input Skydiving!(FortAnimInput_Skydiving:"Fort Anim Input Skydiving Structure (by ref)"=Get SkydivingState() -> (SkydivingState:Fort Anim Input Skydiving Structure)) -> (
    ParaGilderRootOffset:Vector=~
    ParaGliderRootRotationOverride:Rotator=~
    bUseParaGlideRootModifier:bool=~
    bIsSkydivingFromLaunchPad:bool=~
    bIsSkydivingFromBus:bool=~
    bIsInVortex:bool=~
    bIsDiving:bool=~
    bIsParachuteOpen:bool=~
    bIsSkydiving:bool=~
    bIsParachuteLeaning:bool=~
    bIsSkydiveLeaning:bool=~
    bIsLeaning:bool=~
    bIsSkydiveDiveMode:bool=~
    bPlaySkydiveDrift:bool=~
    bSkydiveDriftDelayActive:bool=~
    bSkydiveDriftAnimAllowed:bool=~
    bIsGliderRight:bool=~
    bIsGliderCenter:bool=~
    bIsGliderLeft:bool=~
    bIsGliderForward:bool=~
    bIsGliderBack:bool=~
    bHasDeployedAndIsStandingUpright:bool=~
    LocalAccelForward:float=~
    LocalAccelRight:float=~
    LocalVelocityRight:float=~
    SkydiveAimPitch:float=~
    SkydiveAimYaw:float=~
    DeployChuteAnimRate:float=~
    SkydiveDriftAnimRate:float=~
    SkydiveDriftAnimRateCurrent:float=~
    SkydiveFidgetAnimRate:float=~
    SkydiveFidgetAnimRateCurrent:float=~
    SkydiveAdditiveAlpha:float=~
    SkydiveDriftDelay:float=~
    SkydiveDriftAnim:int
    SkydiveDriftAnimMax:int
    LaunchpadAnim:int
    LocalAccelDir:ESkydivingDirection Enum
    DirectionLast:ESkydivingDirection Enum
    PoseAdditive:Anim Sequence Object Reference
   ) {is=title K2Node_BreakStruct}
   AnimGraphNode_StateMachine_0 = Locomotion() {is=title AnimGraphNode_StateMachine}
   n5 = Apply Additive(
    Base=Apply Additive!(
     Base=@AnimGraphNode_StateMachine_0
     Additive=Blendspace Player 'BS_Elevate_Lean_GLIDER'(X=@n4.LocalAccelRight, Y=@n4.LocalAccelForward) {blendSpace=/BRCosmetics/Animation/Game/MainPlayer/Skydive/ParaGlide/Elevate/BS_Elevate_Lean_GLIDER}
     AlphaCurveName:Name=ParaGlide_Lean
    ) -> (Pose) {alphaInputType=Curve, alphaBoolBlend {blendOption=Linear}, alphaScaleBiasClamp {bias=1, scale=-1}}
    Additive=Blendspace Player 'BS_Elevate_Turn_GLIDER'(X=@n4.SkydiveAimYaw, Y=@n4.LocalAccelForward) {blendSpace=/BRCosmetics/Animation/Game/MainPlayer/Skydive/ParaGlide/Elevate/BS_Elevate_Turn_GLIDER}
   ) {alphaBoolBlend {blendOption=Linear}}
   n6 = Blendspace Player 'BS_Elevate_GLIDER'(X=@n4.SkydiveAimYaw, Y=@n4.SkydiveAimPitch) {blendSpace=/BRCosmetics/Animation/Game/MainPlayer/Skydive/ParaGlide/Elevate/BS_Elevate_GLIDER}
   Save cached pose 'PreAttachMOd'(Pose=Apply Additive!(
     Base=Apply Additive(Base=@n5, Additive=@n6) {alphaBoolBlend {blendOption=Linear}}
     Additive="Sequence Player (additive) 'Elevate_Jitter_Glider'"() {sequence=/BRCosmetics/Animation/Game/MainPlayer/Skydive/ParaGlide/Elevate/Elevate_Jitter_Glider}
     Alpha:float
    ) -> (Pose) {alphaBoolBlend {blendOption=Linear}}) {cacheName=PreAttachMOd}
   AnimGraphNode_StateMachine_1 = Locomotion_GliderItem() {is=title AnimGraphNode_StateMachine}
   Two Way Blend!(A=Use cached pose 'PreAttachMOd'(), B, AlphaCurveName:Name=MatchWeaponBone) -> (Pose) {blendNode {alphaInputType=Curve, alphaBoolBlend {blendOption=Linear}}}
   Get Backpack_BR Transform() -> (Backpack_BR Transform:Transform)
  ]}
 Locomotion {
  entry=Closed
  path=AnimGraph.AnimGraphNode_StateMachine_0.Locomotion
  ids=id
  nodes [
   AnimStateNode_0 = Idle_Open() {is=title AnimStateNode}
   AnimStateNode_2 = Closed() {is=title AnimStateNode, stateEntered {guid=#1, notifyName=ClosedEntered}}
   AnimStateNode_3 = Open_Into() {is=title AnimStateNode, stateEntered {guid=#2, notifyName=DeployEntered}}
   AnimStateTransitionNode_0 = AnimStateTransitionNode() {bAutomaticRuleBasedOnSequencePlayerInState=true, blendMode=Linear, crossfadeDuration=0}
   AnimStateTransitionNode_1 = AnimStateTransitionNode() {crossfadeDuration=0}
   AnimStateTransitionNode_2 = AnimStateTransitionNode() {crossfadeDuration=0}
   AnimStateTransitionNode_3 = AnimStateTransitionNode() {crossfadeDuration=0, priorityOrder=2}
  ]
  links [
   {
    dst=Open_Into
    rule=AnimGraph.AnimGraphNode_StateMachine_0.Locomotion.AnimStateTransitionNode_1.Transition
    src=Closed
    via=AnimStateTransitionNode_1
   }
   {
    dst=Closed
    rule=AnimGraph.AnimGraphNode_StateMachine_0.Locomotion.AnimStateTransitionNode_2.Transition
    src=Idle_Open
    via=AnimStateTransitionNode_2
   }
   {
    dst=Closed
    rule=AnimGraph.AnimGraphNode_StateMachine_0.Locomotion.AnimStateTransitionNode_3.Transition
    src=Open_Into
    via=AnimStateTransitionNode_3
   }
   {
    dst=Idle_Open
    rule=AnimGraph.AnimGraphNode_StateMachine_0.Locomotion.AnimStateTransitionNode_0.Transition
    src=Open_Into
    via=AnimStateTransitionNode_0
   }
  ]
 }
 Idle_Open {
  path=AnimGraph.AnimGraphNode_StateMachine_0.Locomotion.AnimStateNode_0.Idle_Open
  nodes [
   Apply Additive!(
    Base
    Additive
    bAlphaBoolEnabled:bool=Get IsCosmeticPreview!(self:Fort Player Parachute Object Reference=Get OwnerGlider() -> (OwnerGlider:Fort Player Parachute Object Reference)) -> (IsCosmeticPreview:bool)
   ) -> (Pose) {alphaInputType=Bool, alphaBoolBlend {blendOption=Linear}}
   Blend Poses by bool(
    BlendPose_0="Sequence Player (additive) 'Elevate_Glide_Glider'"() {sequence=/BRCosmetics/Animation/Game/MainPlayer/Skydive/ParaGlide/Elevate/Elevate_Glide_Glider}
    bActiveValue=Get IsCosmeticPreview(self=Get OwnerGlider())
   ) {blendType=Linear}
   n1 = Break Fort Anim Input Skydiving(FortAnimInput_Skydiving=Get SkydivingState()) -> (
    ParaGilderRootOffset=~
    ParaGliderRootRotationOverride=~
    bUseParaGlideRootModifier=~
    bIsSkydivingFromLaunchPad=~
    bIsSkydivingFromBus=~
    bIsInVortex=~
    bIsDiving=~
    bIsParachuteOpen=~
    bIsSkydiving=~
    bIsParachuteLeaning=~
    bIsSkydiveLeaning=~
    bIsLeaning=~
    bIsSkydiveDiveMode=~
    bPlaySkydiveDrift=~
    bSkydiveDriftDelayActive=~
    bSkydiveDriftAnimAllowed=~
    bIsGliderRight=~
    bIsGliderCenter=~
    bIsGliderLeft=~
    bIsGliderForward=~
    bIsGliderBack=~
    bHasDeployedAndIsStandingUpright=~
    LocalAccelForward=~
    LocalAccelRight=~
    LocalVelocityRight=~
    SkydiveAimPitch=~
    SkydiveAimYaw=~
    DeployChuteAnimRate=~
    SkydiveDriftAnimRate=~
    SkydiveDriftAnimRateCurrent=~
    SkydiveFidgetAnimRate=~
    SkydiveFidgetAnimRateCurrent=~
    SkydiveAdditiveAlpha=~
    SkydiveDriftDelay=~
   )
   Blendspace Player 'BS_Elevate_Into_WingSuit'(X=@n1.SkydiveAimYaw, Y=@n1.SkydiveAimPitch) {blendSpace=/BRCosmetics/Athena/Items/Cosmetics/Gliders/CustomWingSuitGlider/BS_Elevate_Into_WingSuit}
   Output Animation Pose(Result=Sequence Player 'Elevate_Base_Pose_Glider'() {sequence=/BRCosmetics/Animation/Game/MainPlayer/Skydive/ParaGlide/Elevate/Elevate_Base_Pose_Glider})
  ]
 }
 Closed {
  path=AnimGraph.AnimGraphNode_StateMachine_0.Locomotion.AnimStateNode_2.Closed
  nodes [
   n1 = Get OwnerGlider()
   Output Animation Pose(Result=Blend Poses by bool(
     BlendPose_0=Sequence Player 'Elevate_Close_Glider'() {sequence=/BRCosmetics/Animation/Game/MainPlayer/Skydive/ParaGlide/Elevate/Elevate_Close_Glider}
     BlendPose_1=Sequence Player 'Elevate_Close_Glider'() {sequence=&Elevate_Close_Glider}
     bActiveValue=Get IsCosmeticPreview(self=@n1)
    ) {blendType=Linear})
  ]
 }
 Open_Into {
  path=AnimGraph.AnimGraphNode_StateMachine_0.Locomotion.AnimStateNode_3.Open_Into
  nodes [
   Apply Additive(bAlphaBoolEnabled=Get IsCosmeticPreview(self=Get OwnerGlider())) {alphaInputType=Bool, alphaBoolBlend {blendOption=Linear}}
   n1 = Break Fort Anim Input Skydiving(FortAnimInput_Skydiving=Get SkydivingState()) -> (
    ParaGilderRootOffset=~
    ParaGliderRootRotationOverride=~
    bUseParaGlideRootModifier=~
    bIsSkydivingFromLaunchPad=~
    bIsSkydivingFromBus=~
    bIsInVortex=~
    bIsDiving=~
    bIsParachuteOpen=~
    bIsSkydiving=~
    bIsParachuteLeaning=~
    bIsSkydiveLeaning=~
    bIsLeaning=~
    bIsSkydiveDiveMode=~
    bPlaySkydiveDrift=~
    bSkydiveDriftDelayActive=~
    bSkydiveDriftAnimAllowed=~
    bIsGliderRight=~
    bIsGliderCenter=~
    bIsGliderLeft=~
    bIsGliderForward=~
    bIsGliderBack=~
    bHasDeployedAndIsStandingUpright=~
    LocalAccelForward=~
    LocalAccelRight=~
    LocalVelocityRight=~
    SkydiveAimPitch=~
    SkydiveAimYaw=~
    DeployChuteAnimRate=~
    SkydiveDriftAnimRate=~
    SkydiveDriftAnimRateCurrent=~
    SkydiveFidgetAnimRate=~
    SkydiveFidgetAnimRateCurrent=~
    SkydiveAdditiveAlpha=~
    SkydiveDriftDelay=~
   )
   n2 = Blendspace Player 'BS_Elevate_Into_WingSuit'(X=@n1.SkydiveAimYaw, Y=@n1.SkydiveAimPitch) {bLoop=false, blendSpace=&BS_Elevate_Into_WingSuit}
   n3 = Blendspace Player 'BS_Elevate_Into_NoShieldAnim_WingSuit'(X=@n1.SkydiveAimYaw, Y=@n1.SkydiveAimPitch) {bLoop=false, blendSpace=/BRCosmetics/Athena/Items/Cosmetics/Gliders/CustomWingSuitGlider/BS_Elevate_Into_NoShieldAnim_WingSuit}
   n4 = Get PackMatch() -> (PackMatch:bool)
   n5 = Get OwnerGlider()
   Output Animation Pose(Result=Blend Poses by bool(
     BlendPose_0=Sequence Player 'Elevate_Into_WingSuit'() {bLoopAnimation=false, sequence=/BRCosmetics/Athena/Items/Cosmetics/Gliders/CustomWingSuitGlider/Elevate_Into_WingSuit}
     BlendPose_1=Blend Poses by bool(BlendPose_0=@n3, BlendPose_1=@n2, bActiveValue=@n4) {blendType=Linear}
     bActiveValue=Get IsCosmeticPreview(self=@n5)
    ) {blendType=Linear})
  ]
 }
 Transition {
  path=AnimGraph.AnimGraphNode_StateMachine_0.Locomotion.AnimStateTransitionNode_0.Transition
  nodes [
   Result()
   "Equal (Float)"!(A:double="Get Relevant Anim Time Remaining (Open_Into)"!() -> (ReturnValue:float) {is=title K2Node_AnimGetter}, B:double=0) -> (ReturnValue:bool) {is=title K2Node_CallFunction}
  ]
 }
 Transition {
  path=AnimGraph.AnimGraphNode_StateMachine_0.Locomotion.AnimStateTransitionNode_1.Transition
  nodes [
   n1 = Get SkydivingState()
   Result(bCanEnterTransition=OR Boolean(
     A=Break Fort Anim Input Skydiving(FortAnimInput_Skydiving=@n1) -> (
      ParaGilderRootOffset=~
      ParaGliderRootRotationOverride=~
      bUseParaGlideRootModifier=~
      bIsSkydivingFromLaunchPad=~
      bIsSkydivingFromBus=~
      bIsInVortex=~
      bIsDiving=~
      bIsParachuteOpen=~
      bIsSkydiving=~
      bIsParachuteLeaning=~
      bIsSkydiveLeaning=~
      bIsLeaning=~
      bIsSkydiveDiveMode=~
      bPlaySkydiveDrift=~
      bSkydiveDriftDelayActive=~
      bSkydiveDriftAnimAllowed=~
      bIsGliderRight=~
      bIsGliderCenter=~
      bIsGliderLeft=~
      bIsGliderForward=~
      bIsGliderBack=~
      bHasDeployedAndIsStandingUpright=~
      LocalAccelForward=~
      LocalAccelRight=~
      LocalVelocityRight=~
      SkydiveAimPitch=~
      SkydiveAimYaw=~
      DeployChuteAnimRate=~
      SkydiveDriftAnimRate=~
      SkydiveDriftAnimRateCurrent=~
      SkydiveFidgetAnimRate=~
      SkydiveFidgetAnimRateCurrent=~
      SkydiveAdditiveAlpha=~
      SkydiveDriftDelay=~
     ).bIsParachuteOpen
     B=Get bForceOpen() -> (bForceOpen:bool)
    ))
  ]
 }
 Transition {
  path=AnimGraph.AnimGraphNode_StateMachine_0.Locomotion.AnimStateTransitionNode_2.Transition
  nodes [
   n1 = Get bForceOpen()
   n2 = Break Fort Anim Input Skydiving(FortAnimInput_Skydiving=Get SkydivingState()) -> (
    ParaGilderRootOffset=~
    ParaGliderRootRotationOverride=~
    bUseParaGlideRootModifier=~
    bIsSkydivingFromLaunchPad=~
    bIsSkydivingFromBus=~
    bIsInVortex=~
    bIsDiving=~
    bIsParachuteOpen=~
    bIsSkydiving=~
    bIsParachuteLeaning=~
    bIsSkydiveLeaning=~
    bIsLeaning=~
    bIsSkydiveDiveMode=~
    bPlaySkydiveDrift=~
    bSkydiveDriftDelayActive=~
    bSkydiveDriftAnimAllowed=~
    bIsGliderRight=~
    bIsGliderCenter=~
    bIsGliderLeft=~
    bIsGliderForward=~
    bIsGliderBack=~
    bHasDeployedAndIsStandingUpright=~
    LocalAccelForward=~
    LocalAccelRight=~
    LocalVelocityRight=~
    SkydiveAimPitch=~
    SkydiveAimYaw=~
    DeployChuteAnimRate=~
    SkydiveDriftAnimRate=~
    SkydiveDriftAnimRateCurrent=~
    SkydiveFidgetAnimRate=~
    SkydiveFidgetAnimRateCurrent=~
    SkydiveAdditiveAlpha=~
    SkydiveDriftDelay=~
   )
   Result(bCanEnterTransition=AND Boolean(A=NOT Boolean(A=@n2.bIsParachuteOpen), B=NOT Boolean(A=@n1)))
  ]
 }
 Transition {
  path=AnimGraph.AnimGraphNode_StateMachine_0.Locomotion.AnimStateTransitionNode_3.Transition
  nodes [
   n1 = Get bForceOpen()
   n2 = Break Fort Anim Input Skydiving(FortAnimInput_Skydiving=Get SkydivingState()) -> (
    ParaGilderRootOffset=~
    ParaGliderRootRotationOverride=~
    bUseParaGlideRootModifier=~
    bIsSkydivingFromLaunchPad=~
    bIsSkydivingFromBus=~
    bIsInVortex=~
    bIsDiving=~
    bIsParachuteOpen=~
    bIsSkydiving=~
    bIsParachuteLeaning=~
    bIsSkydiveLeaning=~
    bIsLeaning=~
    bIsSkydiveDiveMode=~
    bPlaySkydiveDrift=~
    bSkydiveDriftDelayActive=~
    bSkydiveDriftAnimAllowed=~
    bIsGliderRight=~
    bIsGliderCenter=~
    bIsGliderLeft=~
    bIsGliderForward=~
    bIsGliderBack=~
    bHasDeployedAndIsStandingUpright=~
    LocalAccelForward=~
    LocalAccelRight=~
    LocalVelocityRight=~
    SkydiveAimPitch=~
    SkydiveAimYaw=~
    DeployChuteAnimRate=~
    SkydiveDriftAnimRate=~
    SkydiveDriftAnimRateCurrent=~
    SkydiveFidgetAnimRate=~
    SkydiveFidgetAnimRateCurrent=~
    SkydiveAdditiveAlpha=~
    SkydiveDriftDelay=~
   )
   Result(bCanEnterTransition=AND Boolean(A=NOT Boolean(A=@n2.bIsParachuteOpen), B=NOT Boolean(A=@n1)))
  ]
 }
 Locomotion_GliderItem {
  entry=Closed
  path=AnimGraph.AnimGraphNode_StateMachine_1.Locomotion_GliderItem
  ids=id
  nodes [
   AnimStateNode_0 = Idle_Open()
   AnimStateNode_1 = ForceOpen_or_GliderItem() {is=title AnimStateNode, stateEntered {guid=#3, notifyName=DeployEntered}}
   AnimStateNode_2 = Closed() {stateEntered {guid=#1, notifyName=ClosedEntered}}
   AnimStateNode_3 = Open_Into() {stateEntered {guid=#2, notifyName=DeployEntered}}
   AnimStateTransitionNode_0 = AnimStateTransitionNode() {bAutomaticRuleBasedOnSequencePlayerInState=true, blendMode=Linear, crossfadeDuration=0}
   AnimStateTransitionNode_1 = AnimStateTransitionNode() {crossfadeDuration=0}
   AnimStateTransitionNode_2 = AnimStateTransitionNode() {crossfadeDuration=0}
   AnimStateTransitionNode_3 = AnimStateTransitionNode() {crossfadeDuration=0, priorityOrder=2}
   AnimStateTransitionNode_4 = AnimStateTransitionNode() {crossfadeDuration=0}
   AnimStateTransitionNode_5 = AnimStateTransitionNode() {bAutomaticRuleBasedOnSequencePlayerInState=true, crossfadeDuration=0}
   AnimStateTransitionNode_6 = AnimStateTransitionNode() {crossfadeDuration=0}
  ]
  links [
   {
    dst=ForceOpen_or_GliderItem
    rule=AnimGraph.AnimGraphNode_StateMachine_1.Locomotion_GliderItem.AnimStateTransitionNode_4.Transition
    src=Closed
    via=AnimStateTransitionNode_4
   }
   {
    dst=Open_Into
    rule=AnimGraph.AnimGraphNode_StateMachine_1.Locomotion_GliderItem.AnimStateTransitionNode_1.Transition
    src=Closed
    via=AnimStateTransitionNode_1
   }
   {
    dst=Closed
    rule=AnimGraph.AnimGraphNode_StateMachine_1.Locomotion_GliderItem.AnimStateTransitionNode_6.Transition
    src=ForceOpen_or_GliderItem
    via=AnimStateTransitionNode_6
   }
   {
    dst=Idle_Open
    rule=AnimGraph.AnimGraphNode_StateMachine_1.Locomotion_GliderItem.AnimStateTransitionNode_5.Transition
    src=ForceOpen_or_GliderItem
    via=AnimStateTransitionNode_5
   }
   {
    dst=Closed
    rule=AnimGraph.AnimGraphNode_StateMachine_1.Locomotion_GliderItem.AnimStateTransitionNode_2.Transition
    src=Idle_Open
    via=AnimStateTransitionNode_2
   }
   {
    dst=Closed
    rule=AnimGraph.AnimGraphNode_StateMachine_1.Locomotion_GliderItem.AnimStateTransitionNode_3.Transition
    src=Open_Into
    via=AnimStateTransitionNode_3
   }
   {
    dst=Idle_Open
    rule=AnimGraph.AnimGraphNode_StateMachine_1.Locomotion_GliderItem.AnimStateTransitionNode_0.Transition
    src=Open_Into
    via=AnimStateTransitionNode_0
   }
  ]
 }
 Idle_Open {
  path=AnimGraph.AnimGraphNode_StateMachine_1.Locomotion_GliderItem.AnimStateNode_0.Idle_Open
  nodes [
   Apply Additive(bAlphaBoolEnabled=Get IsCosmeticPreview(self=Get OwnerGlider())) {alphaInputType=Bool, alphaBoolBlend {blendOption=Linear}}
   n1 = Break Fort Anim Input Skydiving(FortAnimInput_Skydiving=Get SkydivingState()) -> (
    ParaGilderRootOffset=~
    ParaGliderRootRotationOverride=~
    bUseParaGlideRootModifier=~
    bIsSkydivingFromLaunchPad=~
    bIsSkydivingFromBus=~
    bIsInVortex=~
    bIsDiving=~
    bIsParachuteOpen=~
    bIsSkydiving=~
    bIsParachuteLeaning=~
    bIsSkydiveLeaning=~
    bIsLeaning=~
    bIsSkydiveDiveMode=~
    bPlaySkydiveDrift=~
    bSkydiveDriftDelayActive=~
    bSkydiveDriftAnimAllowed=~
    bIsGliderRight=~
    bIsGliderCenter=~
    bIsGliderLeft=~
    bIsGliderForward=~
    bIsGliderBack=~
    bHasDeployedAndIsStandingUpright=~
    LocalAccelForward=~
    LocalAccelRight=~
    LocalVelocityRight=~
    SkydiveAimPitch=~
    SkydiveAimYaw=~
    DeployChuteAnimRate=~
    SkydiveDriftAnimRate=~
    SkydiveDriftAnimRateCurrent=~
    SkydiveFidgetAnimRate=~
    SkydiveFidgetAnimRateCurrent=~
    SkydiveAdditiveAlpha=~
    SkydiveDriftDelay=~
   )
   n2 = Get OwnerGlider()
   Output Animation Pose(Result=Blend Poses by bool(
     BlendPose_0="Sequence Player (additive) 'Elevate_Glide_Glider'"() {sequence=&Elevate_Glide_Glider}
     BlendPose_1=Blendspace Player 'BS_Elevate_GLIDER'(X=@n1.SkydiveAimYaw, Y=@n1.SkydiveAimPitch) {blendSpace=&BS_Elevate_GLIDER}
     bActiveValue=Get IsCosmeticPreview(self=@n2)
    ) {blendType=Linear})
  ]
 }
 ForceOpen_or_GliderItem {
  path=AnimGraph.AnimGraphNode_StateMachine_1.Locomotion_GliderItem.AnimStateNode_1.ForceOpen_or_GliderItem
  nodes [
   n1 = Break Fort Anim Input Skydiving(FortAnimInput_Skydiving=Get SkydivingState()) -> (
    ParaGilderRootOffset=~
    ParaGliderRootRotationOverride=~
    bUseParaGlideRootModifier=~
    bIsSkydivingFromLaunchPad=~
    bIsSkydivingFromBus=~
    bIsInVortex=~
    bIsDiving=~
    bIsParachuteOpen=~
    bIsSkydiving=~
    bIsParachuteLeaning=~
    bIsSkydiveLeaning=~
    bIsLeaning=~
    bIsSkydiveDiveMode=~
    bPlaySkydiveDrift=~
    bSkydiveDriftDelayActive=~
    bSkydiveDriftAnimAllowed=~
    bIsGliderRight=~
    bIsGliderCenter=~
    bIsGliderLeft=~
    bIsGliderForward=~
    bIsGliderBack=~
    bHasDeployedAndIsStandingUpright=~
    LocalAccelForward=~
    LocalAccelRight=~
    LocalVelocityRight=~
    SkydiveAimPitch=~
    SkydiveAimYaw=~
    DeployChuteAnimRate=~
    SkydiveDriftAnimRate=~
    SkydiveDriftAnimRateCurrent=~
    SkydiveFidgetAnimRate=~
    SkydiveFidgetAnimRateCurrent=~
    SkydiveAdditiveAlpha=~
    SkydiveDriftDelay=~
   )
   n2 = Blendspace Player 'BS_Elevate_Into_WingSuit'(X=@n1.SkydiveAimYaw, Y=@n1.SkydiveAimPitch) {bLoop=false, blendSpace=&BS_Elevate_Into_WingSuit, startPosition=0.0063}
   Output Animation Pose(Result=@n2)
  ]
 }
 Closed {
  path=AnimGraph.AnimGraphNode_StateMachine_1.Locomotion_GliderItem.AnimStateNode_2.Closed
  nodes [
   n1 = Get OwnerGlider()
   Output Animation Pose(Result=Blend Poses by bool(
     BlendPose_0=Sequence Player 'Elevate_Close_Glider'() {sequence=&Elevate_Close_Glider}
     BlendPose_1=Sequence Player 'Elevate_Close_Glider'() {sequence=&Elevate_Close_Glider}
     bActiveValue=Get IsCosmeticPreview(self=@n1)
    ) {blendType=Linear})
  ]
 }
 Open_Into {
  path=AnimGraph.AnimGraphNode_StateMachine_1.Locomotion_GliderItem.AnimStateNode_3.Open_Into
  nodes [
   Apply Additive(bAlphaBoolEnabled=Get IsCosmeticPreview(self=Get OwnerGlider())) {alphaInputType=Bool, alphaBoolBlend {blendOption=Linear}}
   n1 = Break Fort Anim Input Skydiving(FortAnimInput_Skydiving=Get SkydivingState()) -> (
    ParaGilderRootOffset=~
    ParaGliderRootRotationOverride=~
    bUseParaGlideRootModifier=~
    bIsSkydivingFromLaunchPad=~
    bIsSkydivingFromBus=~
    bIsInVortex=~
    bIsDiving=~
    bIsParachuteOpen=~
    bIsSkydiving=~
    bIsParachuteLeaning=~
    bIsSkydiveLeaning=~
    bIsLeaning=~
    bIsSkydiveDiveMode=~
    bPlaySkydiveDrift=~
    bSkydiveDriftDelayActive=~
    bSkydiveDriftAnimAllowed=~
    bIsGliderRight=~
    bIsGliderCenter=~
    bIsGliderLeft=~
    bIsGliderForward=~
    bIsGliderBack=~
    bHasDeployedAndIsStandingUpright=~
    LocalAccelForward=~
    LocalAccelRight=~
    LocalVelocityRight=~
    SkydiveAimPitch=~
    SkydiveAimYaw=~
    DeployChuteAnimRate=~
    SkydiveDriftAnimRate=~
    SkydiveDriftAnimRateCurrent=~
    SkydiveFidgetAnimRate=~
    SkydiveFidgetAnimRateCurrent=~
    SkydiveAdditiveAlpha=~
    SkydiveDriftDelay=~
   )
   n2 = Get OwnerGlider()
   Output Animation Pose(Result=Blend Poses by bool(
     BlendPose_0=Sequence Player 'Elevate_Into_WingSuit'() {bLoopAnimation=false, sequence=&Elevate_Into_WingSuit}
     BlendPose_1=Blendspace Player 'BS_Elevate_Into_WingSuit'(X=@n1.SkydiveAimYaw, Y=@n1.SkydiveAimPitch) {bLoop=false, blendSpace=&BS_Elevate_Into_WingSuit}
     bActiveValue=Get IsCosmeticPreview(self=@n2)
    ) {blendType=Linear})
  ]
 }
 Transition {
  path=AnimGraph.AnimGraphNode_StateMachine_1.Locomotion_GliderItem.AnimStateTransitionNode_0.Transition
  nodes [
   Result()
   "Equal (Float)"(A="Get Relevant Anim Time Remaining (Open_Into)"(), B=0)
  ]
 }
 Transition {
  path=AnimGraph.AnimGraphNode_StateMachine_1.Locomotion_GliderItem.AnimStateTransitionNode_1.Transition
  nodes [
   n1 = Break Fort Anim Input Skydiving(FortAnimInput_Skydiving=Get SkydivingState()) -> (
    ParaGilderRootOffset=~
    ParaGliderRootRotationOverride=~
    bUseParaGlideRootModifier=~
    bIsSkydivingFromLaunchPad=~
    bIsSkydivingFromBus=~
    bIsInVortex=~
    bIsDiving=~
    bIsParachuteOpen=~
    bIsSkydiving=~
    bIsParachuteLeaning=~
    bIsSkydiveLeaning=~
    bIsLeaning=~
    bIsSkydiveDiveMode=~
    bPlaySkydiveDrift=~
    bSkydiveDriftDelayActive=~
    bSkydiveDriftAnimAllowed=~
    bIsGliderRight=~
    bIsGliderCenter=~
    bIsGliderLeft=~
    bIsGliderForward=~
    bIsGliderBack=~
    bHasDeployedAndIsStandingUpright=~
    LocalAccelForward=~
    LocalAccelRight=~
    LocalVelocityRight=~
    SkydiveAimPitch=~
    SkydiveAimYaw=~
    DeployChuteAnimRate=~
    SkydiveDriftAnimRate=~
    SkydiveDriftAnimRateCurrent=~
    SkydiveFidgetAnimRate=~
    SkydiveFidgetAnimRateCurrent=~
    SkydiveAdditiveAlpha=~
    SkydiveDriftDelay=~
   )
   n2 = Get IsCosmeticPreview(self=Get OwnerGlider())
   Result(bCanEnterTransition=AND Boolean(
     A=OR Boolean!(A=@n1.bIsSkydivingFromLaunchPad, B=@n1.bIsSkydivingFromBus, C:bool=@n1.bIsInVortex, D:bool=@n2) -> (ReturnValue)
     B=@n1.bIsParachuteOpen
    ))
  ]
 }
 Transition {
  path=AnimGraph.AnimGraphNode_StateMachine_1.Locomotion_GliderItem.AnimStateTransitionNode_2.Transition
  nodes [
   n1 = Get bForceOpen()
   n2 = Break Fort Anim Input Skydiving(FortAnimInput_Skydiving=Get SkydivingState()) -> (
    ParaGilderRootOffset=~
    ParaGliderRootRotationOverride=~
    bUseParaGlideRootModifier=~
    bIsSkydivingFromLaunchPad=~
    bIsSkydivingFromBus=~
    bIsInVortex=~
    bIsDiving=~
    bIsParachuteOpen=~
    bIsSkydiving=~
    bIsParachuteLeaning=~
    bIsSkydiveLeaning=~
    bIsLeaning=~
    bIsSkydiveDiveMode=~
    bPlaySkydiveDrift=~
    bSkydiveDriftDelayActive=~
    bSkydiveDriftAnimAllowed=~
    bIsGliderRight=~
    bIsGliderCenter=~
    bIsGliderLeft=~
    bIsGliderForward=~
    bIsGliderBack=~
    bHasDeployedAndIsStandingUpright=~
    LocalAccelForward=~
    LocalAccelRight=~
    LocalVelocityRight=~
    SkydiveAimPitch=~
    SkydiveAimYaw=~
    DeployChuteAnimRate=~
    SkydiveDriftAnimRate=~
    SkydiveDriftAnimRateCurrent=~
    SkydiveFidgetAnimRate=~
    SkydiveFidgetAnimRateCurrent=~
    SkydiveAdditiveAlpha=~
    SkydiveDriftDelay=~
   )
   Result(bCanEnterTransition=AND Boolean(A=NOT Boolean(A=@n2.bIsParachuteOpen), B=NOT Boolean(A=@n1)))
  ]
 }
 Transition {
  path=AnimGraph.AnimGraphNode_StateMachine_1.Locomotion_GliderItem.AnimStateTransitionNode_3.Transition
  nodes [
   n1 = Get bForceOpen()
   n2 = Break Fort Anim Input Skydiving(FortAnimInput_Skydiving=Get SkydivingState()) -> (
    ParaGilderRootOffset=~
    ParaGliderRootRotationOverride=~
    bUseParaGlideRootModifier=~
    bIsSkydivingFromLaunchPad=~
    bIsSkydivingFromBus=~
    bIsInVortex=~
    bIsDiving=~
    bIsParachuteOpen=~
    bIsSkydiving=~
    bIsParachuteLeaning=~
    bIsSkydiveLeaning=~
    bIsLeaning=~
    bIsSkydiveDiveMode=~
    bPlaySkydiveDrift=~
    bSkydiveDriftDelayActive=~
    bSkydiveDriftAnimAllowed=~
    bIsGliderRight=~
    bIsGliderCenter=~
    bIsGliderLeft=~
    bIsGliderForward=~
    bIsGliderBack=~
    bHasDeployedAndIsStandingUpright=~
    LocalAccelForward=~
    LocalAccelRight=~
    LocalVelocityRight=~
    SkydiveAimPitch=~
    SkydiveAimYaw=~
    DeployChuteAnimRate=~
    SkydiveDriftAnimRate=~
    SkydiveDriftAnimRateCurrent=~
    SkydiveFidgetAnimRate=~
    SkydiveFidgetAnimRateCurrent=~
    SkydiveAdditiveAlpha=~
    SkydiveDriftDelay=~
   )
   Result(bCanEnterTransition=AND Boolean(A=NOT Boolean(A=@n2.bIsParachuteOpen), B=NOT Boolean(A=@n1)))
  ]
 }
 Transition {
  path=AnimGraph.AnimGraphNode_StateMachine_1.Locomotion_GliderItem.AnimStateTransitionNode_4.Transition
  nodes [
   n1 = Break Fort Anim Input Skydiving(FortAnimInput_Skydiving=Get SkydivingState()) -> (
    ParaGilderRootOffset=~
    ParaGliderRootRotationOverride=~
    bUseParaGlideRootModifier=~
    bIsSkydivingFromLaunchPad=~
    bIsSkydivingFromBus=~
    bIsInVortex=~
    bIsDiving=~
    bIsParachuteOpen=~
    bIsSkydiving=~
    bIsParachuteLeaning=~
    bIsSkydiveLeaning=~
    bIsLeaning=~
    bIsSkydiveDiveMode=~
    bPlaySkydiveDrift=~
    bSkydiveDriftDelayActive=~
    bSkydiveDriftAnimAllowed=~
    bIsGliderRight=~
    bIsGliderCenter=~
    bIsGliderLeft=~
    bIsGliderForward=~
    bIsGliderBack=~
    bHasDeployedAndIsStandingUpright=~
    LocalAccelForward=~
    LocalAccelRight=~
    LocalVelocityRight=~
    SkydiveAimPitch=~
    SkydiveAimYaw=~
    DeployChuteAnimRate=~
    SkydiveDriftAnimRate=~
    SkydiveDriftAnimRateCurrent=~
    SkydiveFidgetAnimRate=~
    SkydiveFidgetAnimRateCurrent=~
    SkydiveAdditiveAlpha=~
    SkydiveDriftDelay=~
   )
   Result(bCanEnterTransition=AND Boolean!(
     A=NOT Boolean(A=@n1.bIsSkydivingFromLaunchPad)
     B=NOT Boolean(A=@n1.bIsSkydivingFromBus)
     C:bool=NOT Boolean(A=@n1.bIsInVortex)
     D:bool=Get bForceOpen()
    ) -> (ReturnValue))
  ]
 }
 Transition {path=AnimGraph.AnimGraphNode_StateMachine_1.Locomotion_GliderItem.AnimStateTransitionNode_5.Transition, nodes [Result()]}
 Transition {
  path=AnimGraph.AnimGraphNode_StateMachine_1.Locomotion_GliderItem.AnimStateTransitionNode_6.Transition
  nodes [
   n1 = Get bForceOpen()
   n2 = Break Fort Anim Input Skydiving(FortAnimInput_Skydiving=Get SkydivingState()) -> (
    ParaGilderRootOffset=~
    ParaGliderRootRotationOverride=~
    bUseParaGlideRootModifier=~
    bIsSkydivingFromLaunchPad=~
    bIsSkydivingFromBus=~
    bIsInVortex=~
    bIsDiving=~
    bIsParachuteOpen=~
    bIsSkydiving=~
    bIsParachuteLeaning=~
    bIsSkydiveLeaning=~
    bIsLeaning=~
    bIsSkydiveDiveMode=~
    bPlaySkydiveDrift=~
    bSkydiveDriftDelayActive=~
    bSkydiveDriftAnimAllowed=~
    bIsGliderRight=~
    bIsGliderCenter=~
    bIsGliderLeft=~
    bIsGliderForward=~
    bIsGliderBack=~
    bHasDeployedAndIsStandingUpright=~
    LocalAccelForward=~
    LocalAccelRight=~
    LocalVelocityRight=~
    SkydiveAimPitch=~
    SkydiveAimYaw=~
    DeployChuteAnimRate=~
    SkydiveDriftAnimRate=~
    SkydiveDriftAnimRateCurrent=~
    SkydiveFidgetAnimRate=~
    SkydiveFidgetAnimRateCurrent=~
    SkydiveAdditiveAlpha=~
    SkydiveDriftDelay=~
   )
   Result(bCanEnterTransition=AND Boolean!(A=NOT Boolean(A=@n2.bIsParachuteOpen), B=NOT Boolean(A=@n1)) -> (ReturnValue))
  ]
 }
 EventGraph {nodes [
   Event BlueprintUpdateAnimation()
   > Get Player Info!(execute:Exec) -> (then_1:Exec) {is=title K2Node_MacroInstance}
   > Get Player Transforms!(execute:Exec) -> (then:Exec) {is=title K2Node_MacroInstance}
   AnimNotify_TurnONEngines() {is=title K2Node_Event}
   > SetGliderFullyDeployed!(execute, self:Fort Glider Anim Instance Object Reference, bIsFullyDeployed:bool=true) -> (then)
   n1 = AnimNotify_ClosedEntered() {is=title K2Node_Event}
   n2 = GetOwningComponent()
   HideBoneByName!(
    execute=[@n5.then, @n1.then]
    self:Skinned Mesh Component Object Reference=@n2
    BoneName:Name=Glider_Base
    PhysBodyOption:EPhysBodyOp Enum=PBO_None
   ) -> (then)
   AnimNotify_DeployEntered() {is=title K2Node_Event}
   > n3 = Branch(Condition=Get PackMatch())
   UnHideBoneByName!(execute=[@n4.then, @n3.then], self:Skinned Mesh Component Object Reference=@n2, BoneName:Name=Glider_Base) -> (then)
   n4 = Delay(execute=@n3.else, Duration=0.2)
   Event BlueprintInitializeAnimation()
   > n5 = AddLibrary!(
    execute
    self:Sound Library Subsystem Object Reference=Get SoundLibrarySubsystem() -> (ReturnValue:Sound Library Subsystem Object Reference) {is=title K2Node_GetSubsystem}
    Actor:Actor Object Reference=Get OwnerGlider()
    Library:Sound Library Object Reference=/BRCosmetics/Sounds/Gliders/Elevate/Glider_Elevate_Lib
   ) -> (then)
   Break Fort Anim Input Skydiving(FortAnimInput_Skydiving=Get SkydivingState()) -> (
    ParaGilderRootOffset=~
    ParaGliderRootRotationOverride=~
    bUseParaGlideRootModifier=~
    bIsSkydivingFromLaunchPad=~
    bIsSkydivingFromBus=~
    bIsInVortex=~
    bIsDiving=~
    bIsParachuteOpen=~
    bIsSkydiving=~
    bIsParachuteLeaning=~
    bIsSkydiveLeaning=~
    bIsLeaning=~
    bIsSkydiveDiveMode=~
    bPlaySkydiveDrift=~
    bSkydiveDriftDelayActive=~
    bSkydiveDriftAnimAllowed=~
    bIsGliderRight=~
    bIsGliderCenter=~
    bIsGliderLeft=~
    bIsGliderForward=~
    bIsGliderBack=~
    bHasDeployedAndIsStandingUpright=~
    LocalAccelForward=~
    LocalAccelRight=~
    LocalVelocityRight=~
    SkydiveAimPitch=~
    SkydiveAimYaw=~
    DeployChuteAnimRate=~
    SkydiveDriftAnimRate=~
    SkydiveDriftAnimRateCurrent=~
    SkydiveFidgetAnimRate=~
    SkydiveFidgetAnimRateCurrent=~
    SkydiveAdditiveAlpha=~
    SkydiveDriftDelay=~
   )
  ]}
 GetPlayerInfo {nodes [
   Inputs()
   > n1 = Sequence()
   > n2 = Branch(Condition=Get IsCosmeticPreview(self=Get OwnerGlider()))
   n3 = Get OwnerFortPawn() -> (OwnerFortPawn:Fort Player Pawn Object Reference)
   Set Backpack_BR Transform(
    execute=@n2.else
    Backpack_BR Transform:Transform=GetSocketTransform(self=GetSkeletalMeshForPartType(self=@n3), InSocketName=Backpack_BR, TransformSpace=RTS_World)
   ) -> (Output_Get:Transform)
   > Set WeaponR Transform(WeaponR Transform:Transform=GetSocketTransform(self=Get Mesh(self=@n3), InSocketName=weapon_r, TransformSpace=RTS_World)) -> (Output_Get:Transform)
   n4 = GetSkeletalMeshForPartType(self=@n3, PartType=Backpack)
   > Is Valid(InputObject=@n4)
   n5 = Get SkeletalMesh!(self:Skinned Mesh Component Object Reference=@n4) -> (SkeletalMesh:Skeletal Mesh Object Reference)
   > Set PackMatch(PackMatch:bool="Equal (SoftObjectReference)"(
     A:"Soft Object Reference (by ref)"=Make Soft Reference!(Input:Skeletal Mesh Object Reference=@n5) -> (Output:Skeletal Mesh Soft Object Reference)
     B:"Soft Object Reference (by ref)"=Get Elevate SK Mesh() -> (Elevate SK Mesh:Skeletal Mesh Soft Object Reference)
    ) -> (ReturnValue:bool=~) {is=title K2Node_PromotableOperator}) -> (Output_Get:bool)
   Outputs!(then_1:Exec=@n1.then_1)
  ]}
 GetPlayerTransforms {nodes [
   Inputs()
   n1 = Get OwnerFortPawn()
   > Set Player_Spine05(Player_Spine05:Transform=GetSocketTransform(self=Get Mesh(self=@n1), InSocketName=spine_05, TransformSpace=RTS_World)) -> (Output_Get:Transform)
   > Outputs!(then:Exec)
  ]}
 SetUseAltGliderCam {nodes [
   Inputs()
   > n1 = Branch(Condition=Get IsCosmeticPreview(self=Get OwnerGlider()))
   Outputs(then=[@n2.then, @n1.then])
   n2 = SetUseAltGliderCamera!(
    execute=@n1.else
    self:Fort Player Animation Interface Object Reference=GetAnimationInterface!(self:Fort Player Pawn Object Reference=Get OwnerFortPawn()) -> (ReturnValue:Fort Player Animation Interface Object Reference)
    bUseAltGliderCamera:bool=true
   ) -> (then)
  ]}
]
variables [
 AimPitch {type=number}
 Backpack_BR Transform {type=Transform, default {location=<0 0 0>, rotation=<rot 0 0 0>, scale=<1 1 1>}}
 DeployAnimRate {type=number}
 Elevate SK Mesh {type=SkeletalMesh, default=/BRCosmetics/Accessories/FORT_Backpacks/Backpack_M_MED_Elevate/Meshes/M_MED_Elevate_Pack}
 Example {type=number}
 Forward {type=number}
 IsParachuteOpen {type=boolean}
 PackMatch {type=boolean}
 Player_Spine05 {type=Transform, default {location=<0 0 0>, rotation=<rot 0 0 0>, scale=<1 1 1>}}
 Right {type=number}
 SkydiveYaw {type=number}
 Time {type=number}
 UseDynamicsAlpha {type=number}
 WeaponR Transform {type=Transform, default {location=<0 0 0>, rotation=<rot 0 0 0>, scale=<1 1 1>}}
]
